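# Material function for the F_MED_Ensemble character that adds a stylized, switchable hair highlight to Emissive.
# Polar coordinates from UV0 or UV1 feed a generated band, which thins with view angle and shifts with view and sun direction to fake anisotropy.
# Fresnel-style edge terms, contrast and input masks shape the band.
# The result blends in a specular color, optionally time-of-day adjusted, passed through a LUT function.
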MaterialFunction /BRCosmetics/Characters/Player/Female/Medium/Bodies/F_MED_Ensemble/Materials/MF_F_Ensemble_HairHighlight
graph {nodes [
  n1 = StaticSwitchParameter(True=TextureCoordinate() {coordinateIndex=1}, False=TextureCoordinate()) {
   desc=activate to use the second UV channel for the Polar Coordinates
   group=F_Ensemble_HairHighlight
   parameterName=UseUV1PolarCoord
   sortPriority=0
  }
  n2 = VectorParameter() {
   group=F_Ensemble_HairHighlight
   parameterName=PolarCoord_PositionScale
   channelNames {a=ScaleY, b=ScaleX, g=PositionY, r=PositionX}
   defaultValue=<rgba -0.245 -0.15 3 3>
  }
  n3 = AppendVector(A=@n2, B=@n2)
  Composite!("UV (V2)"=@n1, "PositionScale (V4)"=@n3) {subgraphName=PolarCoordinates}
  Custom() {code="return View.DirectionalLightDirection;", description=DirectionalLightDirection, inputs []}
  FunctionOutput(None=StaticSwitchParameter(True=NamedRerouteUsage(), False=NamedRerouteUsage()) {group=F_Ensemble_HairHighlight, parameterName=UseEnsemble_HairHighlight, sortPriority=0}) {sortPriority=1}
  MaterialFunctionCall() {materialFunction=&CameraDirectionVector}
  n4 = GetMaterialAttributes(None=NamedRerouteUsage()) {attributeGetTypes [#1]}
  n5 = VectorParameter() {group=F_Ensemble_HairHighlight, parameterName=Specular Color, sortPriority=1, defaultValue=<rgba 0.1452 0.1797 0.276 1>}
  n6 = MaterialFunctionCall!(
   In=StaticSwitchParameter(
    True=MaterialFunctionCall!(Color=@n5.RGB) {materialFunction=/Game/Packages/Fortress_SharedMaterials/Anime_Master_Material/MF_TODColorCurve}.Result
    False=@n5.RGB
   ) {group=F_Ensemble_HairHighlight, parameterName=EnsembleHair_UseTODColorCurve}
   LUT
   LUT Multiplier
  ) {materialFunction=/Game/Packages/Fortress_SharedMaterials/Anime_Master_Material/MF_ApplyLUT}
  n7 = ScalarParameter() {defaultValue=0.05, group=F_Ensemble_HairHighlight, parameterName=Band_Width, sliderMax=0.2, sortPriority=4}
  n8 = Transform(None=Constant3Vector() {constant=<rgba 0 0 -1 0>}) {transformSourceType=TRANSFORMSOURCE_View}
  n9 = DotProduct(A=@n8, B=Constant3Vector() {constant=<rgba 0 0 1 0>})
  n10 = LinearInterpolate(
   A=@n7
   B=Multiply(A=@n7, B=ScalarParameter() {defaultValue=1, group=F_Ensemble_HairHighlight, parameterName=Band_Width_Thinning, sliderMax=1, sortPriority=4})
   Alpha=Saturate(None=Multiply(A=@n9, B=@n9))
  )
  n11 = ScalarParameter() {defaultValue=-0.5, group=F_Ensemble_HairHighlight, parameterName=Band_Sharpness, sliderMax=1, sliderMin=-1, sortPriority=4}
  n12 = ScalarParameter() {defaultValue=0.3, group=F_Ensemble_HairHighlight, parameterName=Band_Offset, sliderMax=1, sliderMin=-1, sortPriority=4}
  n13 = MaterialFunctionCall(Float2=NamedRerouteUsage()) {materialFunction=&BreakOutFloat2Components}
  n14 = CollectionParameter() {
   collection=&FortniteMaterialParameters
   parameterName=SunlightDirectionalVector
  }
  n15 = ComponentMask(None=@n14) {b=true, g=true, r=true}
  n16 = DotProduct(A=Normalize(VectorInput=Add(A=@n8, B=@n15)), B=Constant3Vector() {constant=<rgba 0 0 1 0>})
  n17 = Multiply(A=@n16, B=@n16)
  n18 = ScalarParameter() {
   defaultValue=0.1
   desc=Fake anisotropic highlight movement amount
   group=F_Ensemble_HairHighlight
   parameterName=Shift_Amount
   sliderMax=1
   sortPriority=3
  }
  n19 = AppendVector(A=@n13.R, B=Add(A=@n13.G, B=Multiply(A=@n17, B=@n18) {constB=0.1}))
  n20 = NamedRerouteUsage()
  n21 = ScalarParameter() {defaultValue=1, group=F_Ensemble_HairHighlight, parameterName=EdgeGrowth_Fresnel_Exponent, sliderMax=5, sortPriority=5}
  n22 = Subtract(
   A=Multiply(
    A=MaterialFunctionCall!(Width=@n10, Sharpness=@n11, Offset=@n12, Direction Switch, Compare, Input Coordinates=@n19) {materialFunction=/Engine/Functions/Engine_MaterialFunctions03/Procedurals/GeneratedBand}.Result
    B=ScalarParameter() {defaultValue=2.5, group=F_Ensemble_HairHighlight, parameterName=Band_Intensity, sliderMax=5, sortPriority=4}
   ) {constB=2.5}
   B=Saturate(None=Power(Base=@n20, Exp=@n21) {constExponent=1})
  )
  n23 = ScalarParameter() {defaultValue=2, group=F_Ensemble_HairHighlight, parameterName=Contrast, sliderMax=4, sortPriority=6}
  n24 = ScalarParameter() {defaultValue=0.2, group=F_Ensemble_HairHighlight, parameterName=ShadedSpecular_Intensity, sliderMax=1, sortPriority=2}
  n25 = Multiply(
   A=Multiply(
    A=MaterialFunctionCall(In=@n22, Contrast=@n23) {materialFunction=&CheapContrast}.Result
    B=NamedRerouteUsage()
   )
   B=LinearInterpolate(A=@n24, Alpha=NamedRerouteUsage())
  )
  n26 = NamedRerouteUsage()
  n27 = ScalarParameter() {defaultValue=1, group=F_Ensemble_HairHighlight, parameterName=EdgeFadeing_Fresnel_Exponent, sliderMax=5, sortPriority=7}
  n28 = ScalarParameter() {defaultValue=1, group=F_Ensemble_HairHighlight, parameterName=EdgeFadeing_Fresnel_Intensity, sliderMax=5, sortPriority=7}
  n29 = Saturate(None=Multiply(A=Power(Base=@n26, Exp=@n27) {constExponent=1}, B=@n28))
  n30 = Saturate(None=Multiply(A=Multiply(A=@n25, B=@n29), B=NamedRerouteUsage()))
  NamedRerouteDeclaration(None=SetMaterialAttributes!(MaterialAttributes=@n4.MaterialAttributes, EmissiveColor=LinearInterpolate(A=@n4.EmissiveColor, B=@n6.Result, Alpha=@n30)) {attributeSetTypes [#1]}) {name=MA Output, nodeColor=<rgba 0.9922 0 1 1>}
  NamedRerouteDeclaration(None=MaterialFunctionCall(Float3=FunctionInput() {inputName=SCC, sortPriority=3}) {materialFunction=&BreakOutFloat3Components}.B) {name=Sheen, nodeColor=<rgba 0.1016 1 0 1>}
  NamedRerouteDeclaration(None=FunctionInput() {inputName=Shading Curve, inputType=FunctionInput_Scalar, sortPriority=1}) {name=Shading Curve, nodeColor=<rgba 1 0.4571 0 1>}
  NamedRerouteDeclaration(None=DotProduct(A=CameraVectorWS(), B=PixelNormalWS())) {name=Fresnel, nodeColor=<rgba 1 0.1354 0.1453 1>}
  n31 = Constant() {r=-0.5}
  MaterialExpressionReroute_2 = Reroute(None=@n1)
  MaterialExpressionReroute_1 = Reroute(None=@n3)
  n32 = MaterialFunctionCall(Float4=@MaterialExpressionReroute_1."PositionScale (V4)") {materialFunction=&BreakOutFloat4Components}
  n33 = Add(
   A=AppendVector(A=@n31, B=@n31)
   B=Add(A=ComponentMask(None=@MaterialExpressionReroute_2."UV (V2)") {g=true, r=true}, B=AppendVector(A=@n32.R, B=@n32.R))
  )
  n34 = ComponentMask(None=@n33) {g=true}
  n35 = ComponentMask(None=@n33) {r=true}
  n36 = Constant() {r=1}
  n37 = Multiply(A=@n35, B=@n32.B)
  n38 = Multiply(A=@n37, B=@n37)
  n39 = Multiply(A=@n34, B=@n32.A)
  n40 = Multiply(A=@n39, B=@n39)
  n41 = AppendVector(
   A=Divide(
    A=Arctangent2Fast(Y=@n34, X=@n35)
    B=MaterialFunctionCall(Multiplier=@n36) {desc=Radial Gradient, materialFunction=&Pi}.Result
   )
   B=SquareRoot(None=Add(A=@n38, B=@n40))
  )
  MaterialExpressionReroute_0 = Reroute(None=@n41)
  NamedRerouteDeclaration(None=@MaterialExpressionReroute_0.Coord) {name=PolarCoord, nodeColor=<rgba 0 1 0.0391 1>}
  n42 = FunctionInput() {inputName=MA, inputType=FunctionInput_MaterialAttributes}
  NamedRerouteDeclaration(None=GetMaterialAttributes(None=NamedRerouteDeclaration(None=@n42) {name=MA Input, nodeColor=<rgba 1 0 0.014 1>}) {attributeGetTypes [#2]}.Normal) {name=Normal, nodeColor=<rgba 0 0.1145 1 1>}
  NamedRerouteDeclaration(None=FunctionInput() {inputName=InkLine Mask, inputType=FunctionInput_Scalar, sortPriority=2}) {name=InkLine Mask, nodeColor=<rgba 0.0729 0.0729 0.0729 1>}
  PinBase() {reroutePins [
    "UV (V2)" {expression=&self:MaterialExpressionReroute_2}
    "PositionScale (V4)" {expression=&self:MaterialExpressionReroute_1}
   ]}
  PinBase!(Coord=@n41) {reroutePins [Coord {expression=&self:MaterialExpressionReroute_0}]}
 ]}
