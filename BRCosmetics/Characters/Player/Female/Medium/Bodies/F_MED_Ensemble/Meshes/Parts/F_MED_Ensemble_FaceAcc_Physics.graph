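# Physics asset for the F_MED_Ensemble female character's face accessory part, setting up dynamic hair physics.
# Capsules on head, neck, spine, pelvis, arms and backpack provide body collision shapes around the hair bodies.
# Side strands, three short box locks and three long back chains hang from the head through angular constraints.
# Cross constraints tie left and right back chains to the center chain along their lower segments, mostly with free swing.
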PhysicsAsset /BRCosmetics/Characters/Player/Female/Medium/Bodies/F_MED_Ensemble/Meshes/Parts/F_MED_Ensemble_FaceAcc_Physics
bodies {
 backpack_collision [{
   shapeType=Capsule
   length=8.6212
   radius=26.923
   shapeName=backpack_collision_capsule
   center=<3.2691 -18.7525 -0.0106>
   rotation=<rot -89.6958 -172.7492 180>
  }]
 dyn_HairA_l_01 [{shapeType=Box, extentX=7.5, extentY=3, extentZ=3, shapeName=dyn_HairA_l_01_box, center=<-1.9918 0 0>, rotation=<rot 0 0 17.5357>}]
 dyn_HairB_l_01 [{shapeType=Box, extentX=7.5, extentY=3, extentZ=3, shapeName=dyn_HairA_l_01_box, center=<-1.9918 0 0>, rotation=<rot 0 0 1.4567>}]
 dyn_HairC_l_01 [{shapeType=Box, extentX=7.5, extentY=3, extentZ=3, shapeName=dyn_HairA_l_01_box, center=<-1.9918 0 0>, rotation=<rot 0 0 7.4821>}]
 dyn_Hair_BackL_01 [{shapeType=Capsule, length=0.0017, radius=3.0581, shapeName=dyn_Hair_BackL_01_capsule}]
 dyn_Hair_BackL_02 [{
   shapeType=Capsule
   length=4.8415
   radius=6.1327
   shapeName=dyn_Hair_BackL_02_capsule
   center=<0 -2.4043 0.0611>
   rotation=<rot 0 -0.0001 91.4546>
  }]
 dyn_Hair_BackL_03 [{
   shapeType=Capsule
   length=9.6344
   radius=9.1612
   shapeName=dyn_Hair_BackR_04_capsule
   center=<6.8443 -5.3536 -0.8314>
   rotation=<rot 0.6137 -3.7205 -0.0514>
  }]
 dyn_Hair_BackL_04 [{
   shapeType=Capsule
   length=7.3225
   radius=5.6956
   shapeName=dyn_Hair_BackL_04_capsule
   center=<3.4189 -3.653 0.2019>
   rotation=<rot -1.4299 13.1845 89.9668>
  }]
 dyn_Hair_BackL_05 [{
   shapeType=Capsule
   length=5.5399
   radius=5.1272
   shapeName=dyn_Hair_BackL_05_capsule
   center=<0.8953 -2.4721 -0.0049>
   rotation=<rot -0.0001 0 91.3248>
  }]
 dyn_Hair_BackL_06 [{
   shapeType=Capsule
   length=5.5399
   radius=5.1272
   shapeName=dyn_Hair_BackL_06_capsule
   center=<0.8952 -2.7602 0.0018>
   rotation=<rot -0.0001 0 91.3248>
  }]
 dyn_Hair_BackL_07 [{
   shapeType=Capsule
   length=5.5399
   radius=5.1272
   shapeName=dyn_Hair_BackL_06_capsule
   center=<0.8952 -2.7602 0.0018>
   rotation=<rot -0.0001 0 91.3248>
  }]
 dyn_Hair_BackL_08 [{
   shapeType=Capsule
   length=5.5399
   radius=5.1272
   shapeName=dyn_Hair_BackL_06_capsule
   center=<0.8952 -2.7602 0.0018>
   rotation=<rot -0.0001 0 91.3248>
  }]
 dyn_Hair_BackR_01 [{shapeType=Capsule, length=0.0017, radius=3.0581, shapeName=dyn_Hair_BackR_01_capsule}]
 dyn_Hair_BackR_02 [{
   shapeType=Capsule
   length=4.8415
   radius=6.1327
   shapeName=dyn_Hair_BackR_02_capsule
   center=<0 -2.4043 0.0611>
   rotation=<rot 0 -0.0001 91.4546>
  }]
 dyn_Hair_BackR_03 [{
   shapeType=Capsule
   length=9.6344
   radius=9.1612
   shapeName=dyn_Hair_BackR_04_capsule
   center=<6.3941 -3.4387 1.2938>
   rotation=<rot 2.9363 1.3506 -0.6399>
  }]
 dyn_Hair_BackR_04 [{
   shapeType=Capsule
   length=6.8637
   radius=6.1544
   shapeName=dyn_Hair_BackL_04_capsule
   center=<3.3755 -3.4699 0.1818>
   rotation=<rot -1.4299 13.1846 96.0799>
  }]
 dyn_Hair_BackR_05 [{
   shapeType=Capsule
   length=5.3991
   radius=5.268
   shapeName=dyn_Hair_BackR_05_capsule
   center=<0.4718 -0.7181 0.5633>
   rotation=<rot -0.0509 2.1966 91.3237>
  }]
 dyn_Hair_BackR_06 [{
   shapeType=Capsule
   length=5.5399
   radius=5.1272
   shapeName=dyn_Hair_BackR_06_capsule
   center=<0.8953 -3.2918 0.014>
   rotation=<rot -0.0001 0 91.3248>
  }]
 dyn_Hair_BackR_07 [{
   shapeType=Capsule
   length=5.5399
   radius=5.1272
   shapeName=dyn_Hair_Back_06_capsule
   center=<0.9225 -2.5817 -0.0094>
   rotation=<rot 0.0989 -4.2827 91.321>
  }]
 dyn_Hair_BackR_08 [{
   shapeType=Capsule
   length=5.5399
   radius=5.1272
   shapeName=dyn_Hair_Back_06_capsule
   center=<0.9955 -1.6073 -0.0319>
   rotation=<rot 0.0989 -4.2827 91.321>
  }]
 dyn_Hair_Back_01 [{shapeType=Capsule, length=0.0017, radius=3.0581, shapeName=dyn_Hair_Back_01_capsule}]
 dyn_Hair_Back_02 [{
   shapeType=Capsule
   length=4.8415
   radius=6.1327
   shapeName=dyn_Hair_Back_02_capsule
   center=<0 -2.4043 0.0611>
   rotation=<rot 0 -0.0001 91.4546>
  }]
 dyn_Hair_Back_03 [{shapeType=Capsule, length=9.6344, radius=9.1612, shapeName=dyn_Hair_Back_03_capsule, center=<4.8783 -4.8381 0>}]
 dyn_Hair_Back_04 [{
   shapeType=Capsule
   length=5.9422
   radius=5.3078
   shapeName=dyn_Hair_Back_04_capsule
   center=<1.6903 -2.2639 0.4342>
   rotation=<rot -0.1506 6.5248 91.3162>
  }]
 dyn_Hair_Back_05 [{
   shapeType=Capsule
   length=5.5399
   radius=5.1272
   shapeName=dyn_Hair_Back_05_capsule
   center=<0.8954 -2.8087 -0.0041>
   rotation=<rot -0.0001 0 91.3248>
  }]
 dyn_Hair_Back_06 [{
   shapeType=Capsule
   length=5.5399
   radius=5.1272
   shapeName=dyn_Hair_Back_06_capsule
   center=<0.9955 -1.6073 -0.0319>
   rotation=<rot 0.0989 -4.2827 91.321>
  }]
 dyn_Hair_Back_07 [{
   shapeType=Capsule
   length=5.5399
   radius=5.1272
   shapeName=dyn_Hair_Back_06_capsule
   center=<0.9955 -1.6073 -0.0319>
   rotation=<rot 0.0989 -4.2827 91.321>
  }]
 dyn_Hair_Back_08 [{
   shapeType=Capsule
   length=5.5399
   radius=5.1272
   shapeName=dyn_Hair_Back_06_capsule
   center=<0.9955 -1.6073 -0.0319>
   rotation=<rot 0.0989 -4.2827 91.321>
  }]
 dyn_Hair_l_01 [{shapeType=Capsule, length=4.25, radius=2.5, shapeName=dyn_Hair_l_01_capsule, center=<-2.1491 0 -0.3491>, rotation=<rot 80.7727 180 180>}]
 dyn_Hair_l_02 [{shapeType=Capsule, length=4.25, radius=2.5, shapeName=dyn_Hair_l_02_capsule, center=<-2.1491 0 -0.3491>, rotation=<rot 80.7727 180 -180>}]
 dyn_Hair_l_03 [{
   shapeType=Capsule
   length=4.3182
   radius=2.4318
   shapeName=dyn_Hair_l_03_capsule
   center=<-2.1491 0 -0.3491>
   rotation=<rot 80.7727 180 -180>
  }]
 dyn_Hair_l_04 [{
   shapeType=Capsule
   length=4.7117
   radius=2.0383
   shapeName=dyn_Hair_l_04_capsule
   center=<-2.1491 0 -0.3491>
   rotation=<rot 80.7727 180 -180>
  }]
 dyn_Hair_l_base [{shapeType=Capsule, length=0.505, radius=0.505, shapeName=dyn_Hair_l_base_capsule, rotation=<rot 0 -90 0>}]
 dyn_Hair_r_01 [{shapeType=Capsule, length=4.75, radius=1.75, shapeName=dyn_Hair_r_01_capsule, center=<1.6843 0 0.2654>, rotation=<rot 86.6228 180 180>}]
 dyn_Hair_r_02 [{
   shapeType=Capsule
   length=4.75
   radius=1.75
   shapeName=dyn_Hair_r_02_capsule
   center=<1.6843 0 0.2654>
   rotation=<rot 86.6228 -180 -179.9999>
  }]
 dyn_Hair_r_03 [{
   shapeType=Capsule
   length=4.75
   radius=1.75
   shapeName=dyn_Hair_r_03_capsule
   center=<1.695 0 0.1619>
   rotation=<rot 87.4436 -179.9999 -179.9999>
  }]
 dyn_Hair_r_04 [{
   shapeType=Capsule
   length=4.6192
   radius=1.8808
   shapeName=dyn_Hair_r_04_capsule
   center=<3.2271 -1.8858 0.2304>
   rotation=<rot 87.4436 -179.9999 -152.141>
  }]
 dyn_Hair_r_base [{shapeType=Capsule, length=1.75, radius=1.75, shapeName=dyn_Hair_r_base_capsule}]
 head [{shapeType=Capsule, length=7.04, radius=6.96, shapeName=head_capsule, center=<2.0202 0.1733 -0.2592>, rotation=<rot 90 -0.02 -0.0202>}]
 lowerarm_l [{
   shapeType=Capsule
   length=29.75
   radius=8.5
   shapeName=lowerarm_l_capsule
   center=<16.7417 -4.4929 2.2057>
   rotation=<rot -89.6724 179.9922 164.9744>
  }]
 lowerarm_r [{
   shapeType=Capsule
   length=23.6088
   radius=8.2739
   shapeName=lowerarm_r_capsule
   center=<-11.8785 3.1912 0.6114>
   rotation=<rot -89.6724 179.9918 164.9747>
  }]
 neck_02 [{
   shapeType=Capsule
   length=17.2475
   radius=11.2498
   shapeName=neck_02_capsule
   center=<-0.2935 -11.1929 -0.0096>
   rotation=<rot -0.7396 -13.5111 -0.4365>
  }]
 pelvis [{shapeType=Capsule, length=33.7443, radius=22.8397, shapeName=pelvis_capsule, center=<-9.1916 4.9957 0>}]
 spine_03 [{shapeType=Capsule, length=34.3721, radius=13.8579, shapeName=spine_03_capsule, center=<-8.2852 1.5383 0>}]
 spine_04 [{shapeType=Capsule, length=37.373, radius=27.4696, shapeName=spine_04_capsule, center=<5.3592 17.8028 0>}]
 spine_05 [{shapeType=Capsule, length=18.3089, radius=9.1011, shapeName=spine_05_capsule, center=<6.735 -1.3596 0.0001>}]
 upperarm_l [{
   shapeType=Capsule
   length=24.9084
   radius=6.0916
   shapeName=upperarm_l_capsule
   center=<14.7078 -0.0002 1.2361>
   rotation=<rot 84.4941 179.9998 179.9998>
  }]
 upperarm_r [{
   shapeType=Capsule
   length=22.8144
   radius=6.1856
   shapeName=upperarm_r_capsule
   center=<-11.2907 0.6458 0.03>
   rotation=<rot 84.4955 -180 -177.1326>
  }]
}
constraints [
 {bone1Name=dyn_HairA_l_01, bone2Name=head, swing1LimitDegrees=10, swing2LimitDegrees=10, twistLimitDegrees=1}
 {bone1Name=dyn_Hair_l_01, bone2Name=dyn_Hair_l_base, swing1LimitDegrees=20, swing2LimitDegrees=20, twistLimitDegrees=1}
 {bone1Name=dyn_Hair_l_02, bone2Name=dyn_Hair_l_01, swing1LimitDegrees=20, swing2LimitDegrees=20, twistLimitDegrees=1}
 {bone1Name=dyn_Hair_l_03, bone2Name=dyn_Hair_l_02, swing1LimitDegrees=20, swing2LimitDegrees=20, twistLimitDegrees=1}
 {bone1Name=dyn_Hair_l_04, bone2Name=dyn_Hair_l_03, swing1LimitDegrees=20, swing2LimitDegrees=20, twistLimitDegrees=1}
 {bone1Name=dyn_Hair_r_base, bone2Name=head, swing1LimitDegrees=30, swing2LimitDegrees=20, twistLimitDegrees=10}
 {bone1Name=dyn_Hair_r_01, bone2Name=dyn_Hair_r_base, swing1LimitDegrees=10, swing2LimitDegrees=10, twistLimitDegrees=1}
 {bone1Name=dyn_Hair_r_02, bone2Name=dyn_Hair_r_01, swing1LimitDegrees=10, swing2LimitDegrees=10, twistLimitDegrees=1}
 {bone1Name=dyn_Hair_r_03, bone2Name=dyn_Hair_r_02, swing1LimitDegrees=10, swing2LimitDegrees=10, twistLimitDegrees=1}
 {bone1Name=dyn_Hair_r_04, bone2Name=dyn_Hair_r_03, swing1LimitDegrees=10, swing2LimitDegrees=10, twistLimitDegrees=1}
 {bone1Name=dyn_Hair_Back_01, bone2Name=head, swing1LimitDegrees=0.1, swing2LimitDegrees=0.1, twistLimitDegrees=0.1}
 {bone1Name=dyn_Hair_Back_02, bone2Name=dyn_Hair_Back_01, swing1LimitDegrees=5, swing2LimitDegrees=1, twistLimitDegrees=15}
 {bone1Name=dyn_Hair_Back_03, bone2Name=dyn_Hair_Back_02, swing1LimitDegrees=5, swing2LimitDegrees=10, twistLimitDegrees=5}
 {bone1Name=dyn_Hair_Back_04, bone2Name=dyn_Hair_Back_03, swing1LimitDegrees=30, swing2LimitDegrees=10, twistLimitDegrees=10}
 {bone1Name=dyn_Hair_Back_05, bone2Name=dyn_Hair_Back_04, swing1LimitDegrees=30, swing2LimitDegrees=10, twistLimitDegrees=10}
 {bone1Name=dyn_Hair_Back_06, bone2Name=dyn_Hair_Back_05, swing1LimitDegrees=30, swing2LimitDegrees=10, twistLimitDegrees=10}
 {bone1Name=dyn_Hair_BackL_01, bone2Name=head, swing1LimitDegrees=0.1, swing2LimitDegrees=0.1, twistLimitDegrees=0.1}
 {bone1Name=dyn_Hair_BackL_02, bone2Name=dyn_Hair_BackL_01, swing1LimitDegrees=5, swing2LimitDegrees=1, twistLimitDegrees=15}
 {bone1Name=dyn_Hair_BackL_03, bone2Name=dyn_Hair_BackL_02, swing1LimitDegrees=5, swing2LimitDegrees=1, twistLimitDegrees=5}
 {bone1Name=dyn_Hair_BackL_04, bone2Name=dyn_Hair_BackL_03, swing1LimitDegrees=30, swing2LimitDegrees=1, twistLimitDegrees=10}
 {bone1Name=dyn_Hair_BackL_05, bone2Name=dyn_Hair_BackL_04, swing1LimitDegrees=30, swing2LimitDegrees=1, twistLimitDegrees=10}
 {bone1Name=dyn_Hair_BackL_06, bone2Name=dyn_Hair_BackL_05, swing1LimitDegrees=30, swing2LimitDegrees=10, twistLimitDegrees=10}
 {bone1Name=dyn_Hair_BackR_01, bone2Name=head, swing1LimitDegrees=0.1, swing2LimitDegrees=0.1, twistLimitDegrees=0.1}
 {bone1Name=dyn_Hair_BackR_02, bone2Name=dyn_Hair_BackR_01, swing1LimitDegrees=5, swing2LimitDegrees=1, twistLimitDegrees=15}
 {bone1Name=dyn_Hair_BackR_03, bone2Name=dyn_Hair_BackR_02, swing1LimitDegrees=10, swing2LimitDegrees=1, twistLimitDegrees=15}
 {bone1Name=dyn_Hair_BackR_04, bone2Name=dyn_Hair_BackR_03, swing1LimitDegrees=30, swing2LimitDegrees=1, twistLimitDegrees=10}
 {bone1Name=dyn_Hair_BackR_05, bone2Name=dyn_Hair_BackR_04, swing1LimitDegrees=30, swing2LimitDegrees=10, twistLimitDegrees=10}
 {bone1Name=dyn_Hair_BackR_06, bone2Name=dyn_Hair_BackR_05, swing1LimitDegrees=30, swing2LimitDegrees=10, twistLimitDegrees=10}
 {
  bone1Name=dyn_Hair_BackL_04
  bone2Name=dyn_Hair_Back_04
  swing1LimitDegrees=90
  swing1Motion=Free
  swing2LimitDegrees=90
  swing2Motion=Free
  twistLimitDegrees=10
 }
 {bone1Name=dyn_Hair_BackR_04, bone2Name=dyn_Hair_Back_04, swing1LimitDegrees=90, swing2LimitDegrees=90, twistLimitDegrees=10}
 {
  bone1Name=dyn_Hair_BackL_05
  bone2Name=dyn_Hair_Back_05
  swing1LimitDegrees=90
  swing1Motion=Free
  swing2LimitDegrees=90
  swing2Motion=Free
  twistLimitDegrees=10
 }
 {bone1Name=dyn_Hair_BackR_05, bone2Name=dyn_Hair_Back_05, swing1LimitDegrees=90, swing2LimitDegrees=90, twistLimitDegrees=10}
 {
  bone1Name=dyn_Hair_BackL_06
  bone2Name=dyn_Hair_Back_06
  swing1LimitDegrees=90
  swing1Motion=Free
  swing2LimitDegrees=90
  swing2Motion=Free
  twistLimitDegrees=10
 }
 {
  bone1Name=dyn_Hair_BackR_06
  bone2Name=dyn_Hair_Back_06
  swing1LimitDegrees=90
  swing1Motion=Free
  swing2LimitDegrees=90
  swing2Motion=Free
  twistLimitDegrees=10
 }
 {bone1Name=dyn_Hair_BackR_08, bone2Name=dyn_Hair_BackR_07, swing1LimitDegrees=30, swing2LimitDegrees=10, twistLimitDegrees=3}
 {bone1Name=dyn_Hair_BackR_07, bone2Name=dyn_Hair_BackR_06, swing1LimitDegrees=30, swing2LimitDegrees=10, twistLimitDegrees=3}
 {bone1Name=dyn_Hair_BackL_08, bone2Name=dyn_Hair_BackL_07, swing1LimitDegrees=30, swing2LimitDegrees=10, twistLimitDegrees=3}
 {bone1Name=dyn_Hair_BackL_07, bone2Name=dyn_Hair_BackL_06, swing1LimitDegrees=30, swing2LimitDegrees=10, twistLimitDegrees=3}
 {bone1Name=dyn_Hair_Back_08, bone2Name=dyn_Hair_Back_07, swing1LimitDegrees=30, swing2LimitDegrees=10, twistLimitDegrees=3}
 {bone1Name=dyn_Hair_Back_07, bone2Name=dyn_Hair_Back_06, swing1LimitDegrees=30, swing2LimitDegrees=10, twistLimitDegrees=3}
 {
  bone1Name=dyn_Hair_BackR_07
  bone2Name=dyn_Hair_Back_07
  swing1LimitDegrees=90
  swing1Motion=Free
  swing2LimitDegrees=90
  swing2Motion=Free
  twistLimitDegrees=10
 }
 {
  bone1Name=dyn_Hair_BackR_08
  bone2Name=dyn_Hair_Back_08
  swing1LimitDegrees=90
  swing1Motion=Free
  swing2LimitDegrees=90
  swing2Motion=Free
  twistLimitDegrees=10
 }
 {
  bone1Name=dyn_Hair_BackL_07
  bone2Name=dyn_Hair_Back_07
  swing1LimitDegrees=90
  swing1Motion=Free
  swing2LimitDegrees=90
  swing2Motion=Free
  twistLimitDegrees=10
 }
 {
  bone1Name=dyn_Hair_BackL_08
  bone2Name=dyn_Hair_Back_08
  swing1LimitDegrees=90
  swing1Motion=Free
  swing2LimitDegrees=90
  swing2Motion=Free
  twistLimitDegrees=10
 }
 {bone1Name=dyn_HairC_l_01, bone2Name=head, swing1LimitDegrees=10, swing2LimitDegrees=10, twistLimitDegrees=1}
 {bone1Name=dyn_HairB_l_01, bone2Name=head, swing1LimitDegrees=10, swing2LimitDegrees=10, twistLimitDegrees=1}
 {
  bone1Name=dyn_Hair_l_base
  bone2Name=head
  swing1LimitDegrees=45
  swing1Motion=Free
  swing2LimitDegrees=45
  swing2Motion=Free
  twistLimitDegrees=45
  twistMotion=Free
 }
]
